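# Niagara effect for the Aspire prop emote, drawing fluff sprites and flickering gold sparkles around a skeletal mesh.
# One emitter bursts both as separate spawn groups on sphere shapes, with a visibility tag and two sprite renderers.
# Emitter-age curves blend custom fluff and sparkle positions from the mesh toward waveform-wobbled bone offsets, staggered by shape-normal dot product.
# Fluff continues to a second, wider-scattered offset; curl noise and collision are disabled.
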NiagaraSystem /BRCosmetics/Animation/Props/Gadgets_Aspire/FX/NS_Emote_Aspire_Fluff
dependencies {
 usedDataInterfaces [&NiagaraDataInterfaceCurve]
 usedDynamicInputs []
 usedModules [
  &self:SystemSpawnScript.SystemScriptSource.SystemScriptGraph.NiagaraNodeAssignment_1.SetVariables_#1
  &SystemState
  &self:SystemSpawnScript.SystemScriptSource.SystemScriptGraph.NiagaraNodeAssignment_0.SetVariables_#2
 ]
}
emitters {Aspire_Pos {
  emitterUpdateScript [
   &EmitterState(Inactive Response=::NewEnumerator1, Life Cycle Mode=::NewEnumerator1, Loop Behavior=::NewEnumerator1, Loop Duration=10, UseLoopDelay=false)
   &SpawnBurst_Instantaneous(Spawn Count=100, Use Spawn Probability=false, UseLoopCountLimit=false) {module=SpawnBurst_Instantaneous001}
   &SpawnBurst_Instantaneous(
    Spawn Count=&RandomRangeInt(Minimum=35, Maximum=50)
    Spawn Group=1
    Use Spawn Probability=false
    UseLoopCountLimit=false
   )
  ]
  particleSpawnScript [
   /SetVariables(Particles.VisibilityTag=&SpawnGroup()) {isSetParameters=true}
   &InitializeParticle(
    Color Mode=::NewEnumerator0
    Lifetime Max=&MaskFloatBySpawnGroup(Masked Float=7, Passthrough Float=4.1)
    Lifetime Min:NiagaraFloat=&MaskFloatBySpawnGroup(Masked Float=6, Passthrough Float=3.5)
    Lifetime Mode=::NewEnumerator1
    Sprite Size Mode=::NewEnumerator3
    Uniform Sprite Size=16
    UsePositionOffset=false
    Write Material Random=true
   )
   &ShapeLocation(MaskSpawnGroup=true, Sphere Distribution=::NewEnumerator3, Sphere Radius=6.5, UseOwnerScale=true)
   &ShapeLocation(
    MaskSpawnGroup=true
    Override Seed=false
    Spawn Group Mask=1
    Sphere Radius=&RandomRangeFloat(Minimum=12, Maximum=13)
    Sphere Surface Distribution=1
    UseOwnerScale=true
   ) {module=ShapeLocation001}
   &AddVelocity() {enabled=false}
   &CameraOffset(Camera Offset Amount=&MaskFloatBySpawnGroup(Masked Float=0, Passthrough Float=25))
  ]
  particleUpdateScript [
   /SetVariables(
    Particles.Sparkle_Color:LinearColor=&RandomRangeLinearColor(
     Minimum=&MakeLinearColor(
      R:NiagaraFloat=1
      G:NiagaraFloat=0.4369
      B:NiagaraFloat=0
      A:NiagaraFloat=&FloatFromCurve(
       FloatCurve=DataInterface() {propertyValues {Curve {keys [
           <at 0 0 RCIM_Cubic>
           {arriveTangent=1.0526, interpMode=RCIM_Cubic, leaveTangent=1.0526, time=0.15, value=1}
           {tangentMode=RCTM_User, time=0.95, value=1}
           {tangentMode=RCTM_User, time=1, value=0}
          ]}}}
       Scale Curve=&Waveform(
        "Waveform [1]"=::NewEnumerator0
        "[1] Phase Scale"=&RandomRangeFloat(Minimum=1, Maximum=2)
        "[1] Frequency"=1
        "[1] Amplitude Min/Max"=<0 1>
       )
      )
     )
     Maximum=&MakeLinearColor(
      R:NiagaraFloat=1
      G:NiagaraFloat=0.854
      B:NiagaraFloat=0.455
      A:NiagaraFloat=&FloatFromCurve(
       FloatCurve=DataInterface() {propertyValues {Curve {keys [
           <at 0 0 RCIM_Cubic>
           {arriveTangent=1.0526, interpMode=RCIM_Cubic, leaveTangent=1.0526, time=0.15, value=1}
           {tangentMode=RCTM_User, time=0.95, value=1}
           {tangentMode=RCTM_User, time=1, value=0}
          ]}}}
       Scale Curve=&Waveform(
        "Waveform [1]"=::NewEnumerator0
        "[1] Phase Scale"=&RandomRangeFloat(Minimum=1, Maximum=2)
        "[1] Frequency"=1
        "[1] Amplitude Min/Max"=<0 1>
       )
      )
     )
    )
    Particles.Sparkle_Size:Vector2f=&Multiply_Vector2D_ByFloat(Float=&RandomRangeFloat(Minimum=1, Maximum=5), Vector2D=<5 5>)
   ) {isSetParameters=true}
   &ParticleState()
   &SpriteRotationRate(Rotation Rate=&FloatFromCurve(
     FloatCurve=DataInterface() {propertyValues {Curve {keys [<at 0 0 RCIM_Constant>,<at 3.5 1 RCIM_Constant>]}}}
     CurveIndex=$Emitter.Age
     Scale Curve=&RandomRangeFloat(Minimum=250, Maximum=500)
    )) {enabled=false}
   /Niagara/Noise/Waveform(
    EnableExponent1=false
    Global Amplitude Scale:NiagaraFloat=&RandomRangeFloat(Minimum=25, Maximum=100)
    Global Input Position:NiagaraFloat=&ReturnNormalizedExecIndex()
    GlobalClampMax=false
    GlobalClampMin=false
    UseInput1=false
    UsePhase1=false
    "Waveform [1]":ENiagara_Waveforms=::NewEnumerator12
    "[1] Frequency":NiagaraFloat=&RandomRangeFloat(Minimum=1, Maximum=2)
   ) {module=Waveform001}
   &Waveform(
    EnableExponent1=false
    Global Amplitude Scale:NiagaraFloat=&RandomRangeFloat(Minimum=25, Maximum=100)
    Global Input Position:NiagaraFloat=&ReturnNormalizedExecIndex()
    Global Phase:NiagaraFloat=&Add_Float(A=$Emitter.Age, B=0.5)
    GlobalClampMax=false
    GlobalClampMin=false
    UseInput1=false
    UsePhase1=false
    "Waveform [1]":ENiagara_Waveforms=::NewEnumerator12
    "[1] Frequency":NiagaraFloat=&RandomRangeFloat(Minimum=1, Maximum=2)
   ) {module=Waveform002}
   &SpriteFacingAndAlignment(
    Alignment Coordinate Space=::Local
    Facing Coordinate Space=::Local
    Sprite Alignment=&Lerp_Vector(
     A=$Particles.Initial.ShapeLocation.ShapeNormal
     B=&MakeVector(
      X=&Lerp_Float(
       A=&MakeFloatFromVector(Vector=$Particles.ShapeLocation.ShapeNormal, Channel=::NewEnumerator0)
       B=0
       Alpha=/Niagara/DynamicInputs/Helpers/Waveform(
        "Waveform [1]"=::NewEnumerator0
        "[1] Frequency"=&RandomRangeFloat(Minimum=1, Maximum=3)
        "[1] Amplitude Min/Max"=<0 1>
       )
      )
      Y=&MakeFloatFromVector(Vector=$Particles.ShapeLocation.ShapeNormal, Channel=::NewEnumerator1)
      Z=&Abs_Float(Float:NiagaraFloat=&MakeFloatFromVector(Vector=$Particles.ShapeLocation.ShapeNormal, Channel=::NewEnumerator2))
     )
     Alpha=&FloatFromCurve(
      FloatCurve=DataInterface() {propertyValues {Curve {keys [{interpMode=RCIM_Cubic, tangentMode=RCTM_User, time=3.5, value=0}, {interpMode=RCIM_Cubic, tangentMode=RCTM_User, time=4.5, value=1}]}}}
      CurveIndex=&Add_Float(A=$Emitter.Age, B=&RandomRangeFloat(Minimum=0, Maximum=1))
      Scale Curve=&RandomRangeFloat(Minimum=0.2, Maximum=2)
     )
    )
    WriteAlignment=true
    WriteFacing=true
   )
   &SkeletalMeshLocation(
    Bone Sampling Mode=::NewEnumerator4
    MaskSpawnGroup=false
    Orientation Sampling=::NewEnumerator1
    Position Sampling=::NewEnumerator2
    Sampled Position Offset=&Add_Vector(
     A=<0 0 0>
     B=&MaskVectorBySpawnGroup(Masked Vector=$Particles.ShapeLocation.ShapeNormal, Passthrough Vector=$Particles.ShapeLocation001.ShapeNormal)
    )
    Sampled Position Offset Coordinate Space:ENiagaraSkelMeshTransforms=::NewEnumerator0
    Skeletal Mesh=$System.Skeletal Mesh
   )
   &SkeletalMeshLocation(
    Bone / Socket Index=2
    Bone Sampling Mode=::NewEnumerator4
    MaskSpawnGroup=false
    Orientation Sampling=::NewEnumerator1
    Position Sampling=::NewEnumerator2
    Sampled Position Offset=&AddVectorToPosition(
     Position=<0 55 -130>
     Vector=&MakeVector(X=0, Y=$Particles.Waveform002.Waveform, Z=$Particles.Waveform001.Waveform)
    )
    Sampled Position Offset Coordinate Space:ENiagaraSkelMeshTransforms=::NewEnumerator2
    Skeletal Mesh=$System.Skeletal Mesh
   ) {module=SkeletalMeshLocation001}
   &SkeletalMeshLocation(
    Bone / Socket Index=2
    Bone Sampling Mode=::NewEnumerator4
    MaskSpawnGroup=false
    Orientation Sampling=::NewEnumerator1
    Position Sampling=::NewEnumerator2
    Sampled Position Offset=&AddVectorToPosition(
     Position=<0 -115 -150>
     Vector=&MakeVector(X=0, Y=$Particles.Waveform002.Waveform, Z=$Particles.Waveform001.Waveform)
    )
    Sampled Position Offset Coordinate Space:ENiagaraSkelMeshTransforms=::NewEnumerator2
    Skeletal Mesh=$System.Skeletal Mesh
   ) {module=SkeletalMeshLocation002}
   /SetVariables(Particles.DotProduct:NiagaraFloat=/Game/Effects/Niagara/DynamicInputs/Math/FN_RemapDotProduct(
     SystemAxis:Vector3f=$Particles.ShapeLocation.ShapeNormal
     ComparisonAxis:Vector3f=<0 -1 0>
     SourceMin:NiagaraFloat=0
     SourceMax:NiagaraFloat=1
     DestMin:NiagaraFloat=-0.6
     DestMax:NiagaraFloat=0
    )) {isSetParameters=true}
   /SetVariables(Particles.Fluff_Pos:NiagaraPosition=&Lerp_Vector(
     A=&LerpPosition(
      A=&AddVectorToPosition(Position=$Particles.ShapeLocation.ShapePosition, Vector=$Particles.SkeletalMeshLocation.SampledPosition)
      B=&AddVectorToPosition(
       Position=$Particles.SkeletalMeshLocation001.SampledPosition
       Vector=&RandomRangeVector(Minimum=<-15 -15 -15>, Maximum=<15 15 15>)
      )
      Alpha=&FloatFromCurve(
       FloatCurve=DataInterface() {propertyValues {Curve {keys [
           {arriveTangent=2.2305, interpMode=RCIM_Cubic, leaveTangent=2.2305, tangentMode=RCTM_User, time=3.4, value=0}
           {arriveTangent=0.2456, interpMode=RCIM_Cubic, leaveTangent=0.2456, tangentMode=RCTM_User, time=4, value=1}
          ]}}}
       CurveIndex=&Add_Float(A=$Emitter.Age, B=$Particles.DotProduct)
       Scale Curve=0.75
      )
     )
     B=&AddVectorToPosition(
      Position=$Particles.SkeletalMeshLocation002.SampledPosition
      Vector=&RandomRangeVector(Minimum=<-35 -35 -35>, Maximum=<35 35 35>)
     )
     Alpha=&FloatFromCurve(
      FloatCurve=DataInterface() {propertyValues {Curve {keys [
          {arriveTangent=0.6715, interpMode=RCIM_Cubic, leaveTangent=0.6715, tangentMode=RCTM_User, time=4, value=0}
          {interpMode=RCIM_Cubic, leaveTangent=0.3, time=6, value=1}
         ]}}}
      CurveIndex=&Add_Float(A=$Emitter.Age, B=&RandomRangeFloat(Minimum=-0.5, Maximum=0))
      Scale Curve=1.15
     )
    )) {isSetParameters=true}
   /SetVariables(Particles.Sparkle_Pos:NiagaraPosition=&LerpPosition(
     A=&AddVectorToPosition(Position=$Particles.ShapeLocation001.ShapePosition, Vector=$Particles.SkeletalMeshLocation.SampledPosition)
     B=&AddVectorToPosition(
      Position=$Particles.SkeletalMeshLocation001.SampledPosition
      Vector=&RandomRangeVector(Minimum=<-15 -15 -15>, Maximum=<15 15 15>)
     )
     Alpha=&FloatFromCurve(
      FloatCurve=DataInterface() {propertyValues {Curve {keys [
          {arriveTangent=2.2305, interpMode=RCIM_Cubic, leaveTangent=2.2305, tangentMode=RCTM_User, time=3.4, value=0}
          {arriveTangent=0.2456, interpMode=RCIM_Cubic, leaveTangent=0.2456, tangentMode=RCTM_User, time=4, value=1}
         ]}}}
      CurveIndex=&Add_Float(A=$Emitter.Age, B=$Particles.DotProduct)
      Scale Curve=0.75
     )
    )) {isSetParameters=true}
   /SetVariables(Particles.Position=$Particles.SkeletalMeshLocation002.SampledPosition) {enabled=false, isSetParameters=true}
   &Collision(
    Age Colliding Particles=false
    KillOnCollision=false
    Module.Control Rest State Via Surface Normal=false
    Module.EnableMaxCollisionCount=false
    Module.Randomize Collision Normal Vector=false
   ) {enabled=false}
   &CurlNoiseForce(
    Noise Strength=&Lerp_Float(
     A=0
     B=500
     Alpha=&FloatFromCurve(
      FloatCurve=DataInterface() {propertyValues {Curve {keys [{interpMode=RCIM_Cubic, tangentMode=RCTM_User, time=3.5, value=0}, {interpMode=RCIM_Cubic, tangentMode=RCTM_User, time=4.5, value=1}]}}}
      CurveIndex=&Add_Float(A=$Emitter.Age, B=&RandomRangeFloat(Minimum=0, Maximum=1))
     )
    )
    Pan Noise=false
   ) {enabled=false}
   &SolveForcesAndVelocity(Clamp Velocity=false, Limit Acceleration=false)
   &ScaleSpriteSize(Uniform Curve Sprite Scale=DataInterface() {propertyValues {Curve {keys [{interpMode=RCIM_Cubic, tangentMode=RCTM_User, time=0.9, value=1}, {interpMode=RCIM_Cubic, tangentMode=RCTM_User, time=1, value=0}]}}})
   &ScaleColor(
    Scale Alpha=&FloatFromCurve(FloatCurve=DataInterface() {propertyValues {Curve {keys [{interpMode=RCIM_Cubic, tangentMode=RCTM_User, time=0.9, value=1}, {interpMode=RCIM_Cubic, tangentMode=RCTM_User, time=1, value=0}]}}})
    ScaleA=true
    ScaleRGB=true
   )
  ]
  rendererClasses [&NiagaraSpriteRendererProperties]
  renderers [
   {rendererClass=&NiagaraSpriteRendererProperties}
   {rendererClass=&NiagaraSpriteRendererProperties, rendererIndex=1}
  ]
 }}
summary {
 systemName=NS_Emote_Aspire_Fluff
 emitters [{emitterName=Aspire_Pos, rendererClasses [&NiagaraSpriteRendererProperties]}]
 userVariables [User.SKM_Aspire1 {
   type=NiagaraDataInterfaceSkeletalMesh
   defaultValue {
    struct=NiagaraExt_VariableValue_DataInterface
    value {dataInterface=&self:NiagaraDataInterfaceSkeletalMesh_1}
   }
  }]
}
